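# MetaSound preset of the shared stereo boss VO source that supplies idle voice lines for the Kor boss (CherrySmoke).
# It fills the parent's wave array with twenty idle takes, with delay and pitch/volume randomization zeroed and probability at full.
# Playback uses the critical enemy VO class, shared ethereal alerted-VO attenuation and boss combat chatter concurrency.
# A bus send feeds the music volume sidechain bus.
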MetaSoundSource /KorBoss/Assets/Audio/VO/MSP_VO_CherrySmoke_Kor_Idle
preset_parent=/Game/Sounds/Fort_Human_VO_BR/NPCs/Shared/MSS_VO_BR_Boss_Stereo
inputs [
 "Delay Max (Seconds)" {type=Time}
 "Delay Min (Seconds)" {type=Time}
 "Probability (0-1)" {type=Float}
 "Random Pitch Max (Semitones)" {type=Float}
 "Random Pitch Min (Semitones)" {type=Float}
 "Random Volume Change Max (Decibels)" {type=Float}
 "Random Volume Change Min (Decibels)" {type=Float}
 UE.Source.OnPlay {type=Trigger}
 Wave Array {type=WaveAsset:Array}
]
interfaces [UE.OutputFormat.Stereo:1.0, UE.Source.OneShot:1.0, UE.Source:1.1]
outputs [
 "UE.OutputFormat.Stereo.Audio:0" {type=Audio}
 "UE.OutputFormat.Stereo.Audio:1" {type=Audio}
 UE.Source.OneShot.OnFinished {type=Trigger}
]
pages [Default {input_defaults {
   "Delay Max (Seconds)"=0
   "Delay Min (Seconds)"=0
   "Probability (0-1)"=1
   "Random Pitch Max (Semitones)"=0
   "Random Pitch Min (Semitones)"=0
   "Random Volume Change Max (Decibels)"=0
   "Random Volume Change Min (Decibels)"=0
   UE.Source.OnPlay=null
   Wave Array [
    /KorBoss/Assets/Audio/VO/Waves/VO_CherrySmoke_Kor_Idle_01
    /KorBoss/Assets/Audio/VO/Waves/VO_CherrySmoke_Kor_Idle_02
    /KorBoss/Assets/Audio/VO/Waves/VO_CherrySmoke_Kor_Idle_03
    /KorBoss/Assets/Audio/VO/Waves/VO_CherrySmoke_Kor_Idle_04
    /KorBoss/Assets/Audio/VO/Waves/VO_CherrySmoke_Kor_Idle_05
    /KorBoss/Assets/Audio/VO/Waves/VO_CherrySmoke_Kor_Idle_06
    /KorBoss/Assets/Audio/VO/Waves/VO_CherrySmoke_Kor_Idle_07
    /KorBoss/Assets/Audio/VO/Waves/VO_CherrySmoke_Kor_Idle_08
    /KorBoss/Assets/Audio/VO/Waves/VO_CherrySmoke_Kor_Idle_09
    /KorBoss/Assets/Audio/VO/Waves/VO_CherrySmoke_Kor_Idle_10
    /KorBoss/Assets/Audio/VO/Waves/VO_CherrySmoke_Kor_Idle_11
    /KorBoss/Assets/Audio/VO/Waves/VO_CherrySmoke_Kor_Idle_12
    /KorBoss/Assets/Audio/VO/Waves/VO_CherrySmoke_Kor_Idle_13
    /KorBoss/Assets/Audio/VO/Waves/VO_CherrySmoke_Kor_Idle_14
    /KorBoss/Assets/Audio/VO/Waves/VO_CherrySmoke_Kor_Idle_15
    /KorBoss/Assets/Audio/VO/Waves/VO_CherrySmoke_Kor_Idle_16
    /KorBoss/Assets/Audio/VO/Waves/VO_CherrySmoke_Kor_Idle_17
    /KorBoss/Assets/Audio/VO/Waves/VO_CherrySmoke_Kor_Idle_18
    /KorBoss/Assets/Audio/VO/Waves/VO_CherrySmoke_Kor_Idle_19
    /KorBoss/Assets/Audio/VO/Waves/VO_CherrySmoke_Kor_Idle_20
   ]
  }}]
properties {
 assetImportData=&self:AssetImportData
 attenuationSettings=/Game/Sounds/Fort_Human_VO_BR/NPCs/Shared/NPC_VO_Alerted_Ethereal_Stereo_Att
 maxDistance=12500
 numChannels=2
 outputFormat=Stereo
 sampleRate=48000
 soundClassObject=&Fort_VO_Enemy_Critical
 volume=2.2
 busSends [{
   audioBus=&AB_Sidechain_FortMusic_Vol
   maxSendDistance=13000
   maxSendLevel=0
   minSendDistance=6000
   minSendLevel=1
   sendLevel=0.3
  }]
 concurrencyOverrides {bLimitToOwner=true, maxCount=1, resolutionRule=PreventNew}
 concurrencySet [&NPC_Boss_Combat_Chatter_Conc]
}
